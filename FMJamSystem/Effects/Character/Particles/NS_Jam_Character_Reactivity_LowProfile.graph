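# Low-profile Niagara effect for Jam characters: audio-reactive rings, glows and particles attached to the skeletal mesh, tinted by track color.
# Mesh emitters draw waves, rings and cylinder glows; sprite emitters add a rising burst and drifting squares.
# Four-band reactivity and loop amplitude drive wave and wavy-ring material parameters, with time-of-day scaling some alphas.
# IsDrum and IsAutoJammer floats act as lerp switches for size, offset and pop-animation curves.
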NiagaraSystem /FMJamSystem/Effects/Character/Particles/NS_Jam_Character_Reactivity_LowProfile
dependencies {
 usedDataInterfaces [
  &NiagaraDataInterfaceCurve
  &NiagaraDataInterfaceVectorCurve
  &NiagaraDataInterfaceCamera
  &NiagaraDataInterfaceVector2DCurve
  &NiagaraDataInterfaceSkeletalMesh
 ]
 usedDynamicInputs [&ReturnCameraVector, &MakeQuatFromAxisAngle]
 usedModules [
  &self:SystemSpawnScript.SystemScriptSource.SystemScriptGraph.NiagaraNodeAssignment_0.SetVariables_#1
  &self:SystemSpawnScript.SystemScriptSource.SystemScriptGraph.NiagaraNodeAssignment_2.SetVariables_#2
  &self:SystemSpawnScript.SystemScriptSource.SystemScriptGraph.NiagaraNodeAssignment_1.SetVariables_#3
  &SystemState
 ]
}
emitters {
 Burst {
  emitterUpdateScript [
   &EmitterState(Life Cycle Mode=::NewEnumerator1, Loop Behavior=::NewEnumerator1, Loop Duration=1, UseLoopDelay=false)
   &SpawnBurst_Instantaneous(Spawn Count=25, Use Spawn Probability=false, UseLoopCountLimit=false)
   &SpawnRate(SpawnRate=&Lerp_Float(A=20, B=40, Alpha=$User.IsDrum), Use Spawn Probability=false) {enabled=false}
  ]
  particleSpawnScript [
   &InitializeParticle(
    Color=$User.TrackColor
    Lifetime Max=1
    Lifetime Min=0.5
    Lifetime Mode=::NewEnumerator1
    Mass Mode=::NewEnumerator3
    Sprite Size=&RandomRangeVector2D(Minimum=<1 1>, Maximum=<2 2>)
    Sprite Size Mode=::NewEnumerator4
    UsePositionOffset=true
    Write Material Random=true
   )
   &AddVelocity(Velocity=&MakeVector(X=0, Y=0, Z=&RandomRangeFloat(Minimum=0, Maximum=700)))
  ]
  particleUpdateScript [
   &ParticleState()
   &ScaleSpriteSize(Uniform Curve Sprite Scale=DataInterface() {propertyValues {Curve {keys [
        {arriveTangent=29.7977, interpMode=RCIM_Cubic, leaveTangent=29.7977, tangentMode=RCTM_User, time=0, value=1}
        {interpMode=RCIM_Cubic, tangentMode=RCTM_User, time=0.2, value=1}
       ]}}}) {module=ScaleSpriteSize001}
   &ShapeLocation(
    Apply To Particle Position=::NewEnumerator1
    MaskSpawnGroup=false
    Override Seed=false
    Ring Radius=&Lerp_Float(A=&Lerp_Float(A=50, B=135, Alpha=$User.IsDrum), B=40, Alpha=$User.IsAutoJammer)
    Rotation Angle=0.25
    Rotation Axis=<0 0 1>
    Rotation Mode=::NewEnumerator1
    Shape Origin=&VectorFromFloat(Value=0)
    Shape Primitive=::NewEnumerator4
    UseOwnerScale=true
   )
   &SkeletalMeshLocation(
    MaskSpawnGroup=false
    Mesh Sampling Type=::NewEnumerator5
    Orientation Sampling=::NewEnumerator1
    Position Sampling=::NewEnumerator1
    Sampled Position Offset=&AddVectorToPosition(
     Position=$StackContext.ShapeLocation.ShapePosition
     Vector=&Lerp_Vector(A=&Lerp_Vector(A=<0 0 30>, B=<50 0 30>, Alpha=$User.IsDrum), B=<0 0 30>, Alpha=$User.IsAutoJammer)
    )
    Skeletal Mesh=$System.Skeletal Mesh
    Socket Sampling Mode=::NewEnumerator5
    Source Movement Scale:Vector3f=<0 0 0>
   )
   &ScaleSpriteSize(
    Non-Uniform Curve Sprite Scale=DataInterface() {propertyValues {
      XCurve {keys [{interpMode=RCIM_Cubic, tangentMode=RCTM_User, time=0, value=1}, {interpMode=RCIM_Cubic, tangentMode=RCTM_User, time=1, value=0}]}
      YCurve {keys [{interpMode=RCIM_Cubic, tangentMode=RCTM_User, time=0, value=1}, {interpMode=RCIM_Cubic, tangentMode=RCTM_User, time=1, value=0}]}
     }}
    Scale Sprite Size Mode=::NewEnumerator3
   )
   &ScaleColor(
    Scale Alpha=&FloatFromCurve(FloatCurve=DataInterface() {propertyValues {Curve {keys [{interpMode=RCIM_Cubic, tangentMode=RCTM_User, time=0, value=1}, {interpMode=RCIM_Cubic, tangentMode=RCTM_User, time=1, value=0}]}}})
    Scale RGB=<5 5 5>
    ScaleA=true
    ScaleRGB=true
   )
   &Drag(Drag=5, Use Linear Drag=true, Use Rotational Drag=true)
   &CurlNoiseForce(Bias Noise=false, Mask Curl Noise=true, Noise Strength=500, Pan Noise=false)
   &SolveForcesAndVelocity(Clamp Velocity=false, Limit Acceleration=false)
   &ScaleSpriteSizeBySpeed(Max Scale Factor=<1 25>)
  ]
  rendererClasses [&NiagaraSpriteRendererProperties]
  renderers [{rendererClass=&NiagaraSpriteRendererProperties}]
 }
 CylinderGlows {
  emitterUpdateScript [
   &EmitterState(Life Cycle Mode=::NewEnumerator1, Loop Behavior=::NewEnumerator1, Loop Duration Mode=::NewEnumerator1, UseLoopDelay=false)
   &SpawnBurst_Instantaneous(Use Spawn Probability=false, UseLoopCountLimit=false)
  ]
  particleSpawnScript [&InitializeParticle(
    Color=$User.TrackColor
    Mesh Scale=&Lerp_Vector(
     A=&Lerp_Vector(A=<1.5 1.5 1.5>, B=<2.95 2.95 1.5>, Alpha=$User.IsDrum)
     B=<1.09 1.09 1.5>
     Alpha=$User.IsAutoJammer
    )
    Mesh Scale Mode=::NewEnumerator4
    UsePositionOffset=false
    Write Material Random=true
   )]
  particleUpdateScript [
   &ParticleState(Kill Particles When Lifetime Has Elapsed=false, Let Infinitely Lived Particles Die When Emitter Deactivates=true)
   &SkeletalMeshLocation(
    MaskSpawnGroup=false
    Mesh Sampling Type=::NewEnumerator5
    Orientation Sampling=::NewEnumerator1
    Sampled Position Offset=&Lerp_Vector(A=&Lerp_Vector(A=<0 0 -7>, B=<50 0 -5.5>, Alpha=$User.IsDrum), B=<0 0 -7>, Alpha=$User.IsAutoJammer)
    Skeletal Mesh=$System.Skeletal Mesh
    Socket Sampling Mode=::NewEnumerator5
   ) {module=SkeletalMeshLocation003}
   &ScaleColor(
    Scale Alpha=&FNTimeOfDayScalar(Night Scale=0.8, Morning Scale=0.75, Evening Scale=0.75)
    Scale RGB=&VectorFromFloat(Value=&FloatFromCurve(FloatCurve=DataInterface() {propertyValues {Curve {keys [
          {arriveTangent=9.0725, interpMode=RCIM_Cubic, leaveTangent=9.0725, tangentMode=RCTM_User, time=0, value=1}
          {arriveTangent=-2.5249, interpMode=RCIM_Cubic, leaveTangent=-2.5249, tangentMode=RCTM_User, time=0.05, value=2}
          {arriveTangent=-0.0405, interpMode=RCIM_Cubic, leaveTangent=-0.0405, tangentMode=RCTM_User, time=0.5, value=1}
         ]}}}))
    ScaleA=true
    ScaleRGB=true
   )
   &ScaleMeshSize(Scale Factor=&Lerp_Vector(
     A=&MakeVector(
      X=&Lerp_Float(
       A=&FloatFromCurve(FloatCurve=$System.AnimCurve)
       B=&FloatFromCurve(FloatCurve=$System.AnimCurveDrum)
       Alpha=$User.IsDrum
      )
      Y=&Lerp_Float(
       A=&FloatFromCurve(FloatCurve=$System.AnimCurve)
       B=&FloatFromCurve(FloatCurve=$System.AnimCurveDrum)
       Alpha=$User.IsDrum
      )
      Z=1
     )
     B=&MakeVector(
      X=&FloatFromCurve(FloatCurve=$System.AnimCurve)
      Y=&FloatFromCurve(FloatCurve=$System.AnimCurve)
      Z=1
     )
     Alpha=$User.IsAutoJammer
    ))
   /SetVariables(Particles.GlowUpwardsPosition:NiagaraPosition=&AddVectorToPosition(Position=$Particles.Position, Vector=<0 0 33>)) {isSetParameters=true}
   /SetVariables(Particles.GlowUpwardsColor:LinearColor=&MultiplyLinearColor(
     Linear Color=$User.TrackColor
     Scale Factor=&MakeFloat4FromVectorAndFloat(
      "Vector (RGB)"=&VectorFromFloat(Value=&FloatFromCurve(FloatCurve=DataInterface() {propertyValues {Curve {keys [
            {arriveTangent=9.0725, interpMode=RCIM_Cubic, leaveTangent=9.0725, tangentMode=RCTM_User, time=0, value=1}
            {arriveTangent=-2.5249, interpMode=RCIM_Cubic, leaveTangent=-2.5249, tangentMode=RCTM_User, time=0.05, value=5}
            {arriveTangent=-0.0405, interpMode=RCIM_Cubic, leaveTangent=-0.0405, tangentMode=RCTM_User, time=0.5, value=1}
           ]}}}))
      "Float (Alpha)"=&FNTimeOfDayScalar(Night Scale=0.8, Morning Scale=0.75, Evening Scale=0.75)
     )
    )) {isSetParameters=true}
   /SetVariables(Particles.GlowUpwardsScale:Vector3f=&Multiply_Vector(
     A=$Particles.Initial.Scale
     B=&Lerp_Vector(
      A=&MakeVector(
       X=&Lerp_Float(
        A=&FloatFromCurve(FloatCurve=$System.AnimCurve)
        B=&FloatFromCurve(FloatCurve=$System.AnimCurveDrum)
        Alpha=$User.IsDrum
       )
       Y=&Lerp_Float(
        A=&FloatFromCurve(FloatCurve=$System.AnimCurve)
        B=&FloatFromCurve(FloatCurve=$System.AnimCurveDrum)
        Alpha=$User.IsDrum
       )
       Z=&FloatFromCurve(
        FloatCurve=DataInterface() {propertyValues {Curve {keys [
            {arriveTangent=24.9707, interpMode=RCIM_Cubic, leaveTangent=24.9707, tangentMode=RCTM_User, time=0, value=1}
            {arriveTangent=-6.1044, interpMode=RCIM_Cubic, leaveTangent=-6.1044, tangentMode=RCTM_User, time=0.1, value=7}
            {interpMode=RCIM_Cubic, tangentMode=RCTM_User, time=0.5, value=1}
           ]}}}
        Scale Curve=0.5
       )
      )
      B=&MakeVector(
       X=&FloatFromCurve(FloatCurve=$System.AnimCurve)
       Y=&FloatFromCurve(FloatCurve=$System.AnimCurve)
       Z=&FloatFromCurve(
        FloatCurve=DataInterface() {propertyValues {Curve {keys [
            {arriveTangent=24.9707, interpMode=RCIM_Cubic, leaveTangent=24.9707, tangentMode=RCTM_User, time=0, value=1}
            {arriveTangent=-6.1044, interpMode=RCIM_Cubic, leaveTangent=-6.1044, tangentMode=RCTM_User, time=0.1, value=7}
            {interpMode=RCIM_Cubic, tangentMode=RCTM_User, time=0.5, value=1}
           ]}}}
        Scale Curve=0.5
       )
      )
      Alpha=$User.IsAutoJammer
     )
    )) {isSetParameters=true}
  ]
  rendererClasses [&NiagaraMeshRendererProperties]
  renderers [
   {rendererClass=&NiagaraMeshRendererProperties, rendererIndex=0}
   {rendererClass=&NiagaraMeshRendererProperties, rendererIndex=1}
  ]
 }
 ReactiveWave {
  emitterUpdateScript [
   &EmitterState(Life Cycle Mode=::NewEnumerator1, Loop Behavior=::NewEnumerator1, Loop Duration Mode=::NewEnumerator1, UseLoopDelay=false)
   &SpawnBurst_Instantaneous(Use Spawn Probability=false, UseLoopCountLimit=false)
  ]
  particleSpawnScript [&InitializeParticle(
    Color=$User.TrackColor
    Mesh Scale=&Lerp_Vector(A=&Lerp_Vector(A=<1 1 1>, B=<2 2 1>, Alpha=$User.IsDrum), B=<0.75 0.75 1>, Alpha=$User.IsAutoJammer)
    Mesh Scale Mode=::NewEnumerator4
    Sprite Size Mode=::NewEnumerator3
    Uniform Sprite Size=&Lerp_Float(A=200, B=400, Alpha=$User.IsDrum)
    UsePositionOffset=false
    Write Material Random=true
   )]
  particleUpdateScript [
   &ParticleState(Kill Particles When Lifetime Has Elapsed=false, Let Infinitely Lived Particles Die When Emitter Deactivates=true)
   &UpdateMeshOrientation(
    Facing Direction=<0 -1 0>
    Facing Interpolation=false
    Orientation Method=::NewEnumerator1
    Override Quat=false
    TransformFacing=false
    TransformUp=false
   )
   &SkeletalMeshLocation(
    MaskSpawnGroup=false
    Mesh Sampling Type=::NewEnumerator5
    Orientation Sampling=::NewEnumerator1
    Sampled Position Offset=&Lerp_Vector(A=&Lerp_Vector(A=<0 0 30>, B=<50 0 30>, Alpha=$User.IsDrum), B=<0 0 30>, Alpha=$User.IsAutoJammer)
    Skeletal Mesh=$System.Skeletal Mesh
    Socket Sampling Mode=::NewEnumerator5
   ) {module=SkeletalMeshLocation003}
   &DynamicMaterialParameters(
    Index 0 Param 1=$User.IsDrum
    Index 0 Param 2=&Multiply_Float(A=$User.LoopAmplitudeMultiply, B=2)
    Index 0 Param 3=&FloatFromCurve(FloatCurve=DataInterface() {propertyValues {Curve {keys [
         {arriveTangent=0.3779, interpMode=RCIM_Cubic, leaveTangent=0.3779, tangentMode=RCTM_User, time=0, value=1}
         {arriveTangent=0.0149, interpMode=RCIM_Cubic, leaveTangent=0.0149, tangentMode=RCTM_User, time=0.1, value=0}
         {interpMode=RCIM_Cubic, tangentMode=RCTM_Break, time=0.6, value=0}
         {arriveTangent=-0.0405, interpMode=RCIM_Cubic, leaveTangent=-0.0405, tangentMode=RCTM_User, time=0.8, value=1}
        ]}}})
    Index 0 Param 4=&FloatFromCurve(
     FloatCurve=DataInterface() {propertyValues {Curve {keys [
         {arriveTangent=0.3779, interpMode=RCIM_Cubic, leaveTangent=0.3779, tangentMode=RCTM_User, time=0, value=0}
         {arriveTangent=0.0149, interpMode=RCIM_Cubic, leaveTangent=0.0149, tangentMode=RCTM_User, time=0.1, value=1}
         {arriveTangent=-0.0405, interpMode=RCIM_Cubic, leaveTangent=-0.0405, tangentMode=RCTM_User, time=0.6, value=0}
        ]}}}
     Scale Curve=2
    )
    Index 1 Param 1=&MakeFloatFromVector4(Vector 4=$User.4BandReactivity, Channel=::NewEnumerator0)
    Index 1 Param 2=&MakeFloatFromVector4(Vector 4=$User.4BandReactivity, Channel=::NewEnumerator1)
    Index 1 Param 3=&MakeFloatFromVector4(Vector 4=$User.4BandReactivity, Channel=::NewEnumerator0)
    Index 1 Param 4=&MakeFloatFromVector4(Vector 4=$User.4BandReactivity, Channel=::NewEnumerator3)
    Module.Index0Param0Write=true
    Module.Index0Param1Write=true
    Module.Index0Param2Write=true
    Module.Index0Param3Write=true
    Module.Index1Param0Write=true
    Module.Index1Param1Write=true
    Module.Index1Param2Write=true
    Module.Index1Param3Write=true
    Write Parameter Index 1=true
   )
   &ScaleColor(
    Scale Alpha=&FNTimeOfDayScalar(Night Scale=0.65, Morning Scale=0.65, Day Scale=0.75, Evening Scale=0.5)
    Scale RGB=&VectorFromFloat(Value=&FloatFromCurve(FloatCurve=DataInterface() {propertyValues {Curve {keys [
          {arriveTangent=9.0725, interpMode=RCIM_Cubic, leaveTangent=9.0725, tangentMode=RCTM_User, time=0, value=1}
          {arriveTangent=-2.5249, interpMode=RCIM_Cubic, leaveTangent=-2.5249, tangentMode=RCTM_User, time=0.1408, value=1.5}
          {arriveTangent=-0.0405, interpMode=RCIM_Cubic, leaveTangent=-0.0405, tangentMode=RCTM_User, time=1, value=1}
         ]}}}))
    ScaleA=true
    ScaleRGB=true
   )
   &ScaleMeshSize(Scale Factor=&Lerp_Vector(
     A=&MakeVector(
      X=&Lerp_Float(
       A=&FloatFromCurve(FloatCurve=$System.AnimCurve)
       B=&FloatFromCurve(FloatCurve=$System.AnimCurveDrum)
       Alpha=$User.IsDrum
      )
      Y=&Lerp_Float(
       A=&FloatFromCurve(FloatCurve=$System.AnimCurve)
       B=&FloatFromCurve(FloatCurve=$System.AnimCurveDrum)
       Alpha=$User.IsDrum
      )
      Z=1
     )
     B=&MakeVector(
      X=&FloatFromCurve(FloatCurve=$System.AnimCurve)
      Y=&FloatFromCurve(FloatCurve=$System.AnimCurve)
      Z=1
     )
     Alpha=$User.IsAutoJammer
    ))
  ]
  rendererClasses [&NiagaraMeshRendererProperties]
  renderers [{rendererClass=&NiagaraMeshRendererProperties, rendererIndex=0}]
 }
 SquareParticles {
  emitterUpdateScript [
   &EmitterState(Life Cycle Mode=::NewEnumerator1, Loop Duration=1, UseLoopDelay=false)
   &SpawnRate(SpawnRate=20, Use Spawn Probability=false)
  ]
  particleSpawnScript [
   &InitializeParticle(
    Color=$User.TrackColor
    Lifetime Min=0.5
    Lifetime Mode=::NewEnumerator1
    Mass Mode=::NewEnumerator3
    Sprite Size=<2 2>
    Sprite Size Mode=::NewEnumerator4
    UsePositionOffset=true
    Write Material Random=true
   )
   &AddVelocity(Velocity=&MakeVector(X=0, Y=0, Z=-20))
  ]
  particleUpdateScript [
   &ParticleState()
   &ShapeLocation(
    Apply To Particle Position=::NewEnumerator1
    MaskSpawnGroup=false
    Override Seed=false
    Ring Radius=&Lerp_Float(A=&Lerp_Float(A=70, B=135, Alpha=$User.IsDrum), B=50, Alpha=$User.IsAutoJammer)
    Shape Origin=&VectorFromFloat(Value=0)
    Shape Primitive=::NewEnumerator4
    UseOwnerScale=true
   )
   &SkeletalMeshLocation(
    MaskSpawnGroup=false
    Mesh Sampling Type=::NewEnumerator5
    Orientation Sampling=::NewEnumerator1
    Position Sampling=::NewEnumerator1
    Sampled Position Offset=&AddVectorToPosition(
     Position=$StackContext.ShapeLocation.ShapePosition
     Vector=&Lerp_Vector(A=&Lerp_Vector(A=<0 0 30>, B=<50 0 30>, Alpha=$User.IsDrum), B=<0 0 30>, Alpha=$User.IsAutoJammer)
    )
    Skeletal Mesh=$System.Skeletal Mesh
    Socket Sampling Mode=::NewEnumerator5
    Source Movement Scale:Vector3f=<0 0 0>
   ) {module=SkeletalMeshLocation003}
   &ScaleSpriteSize(
    Non-Uniform Curve Sprite Scale=DataInterface() {propertyValues {
      XCurve {keys [{interpMode=RCIM_Cubic, tangentMode=RCTM_User, time=0, value=1}, {interpMode=RCIM_Cubic, tangentMode=RCTM_User, time=1, value=0}]}
      YCurve {keys [{interpMode=RCIM_Cubic, tangentMode=RCTM_User, time=0, value=1}, {interpMode=RCIM_Cubic, tangentMode=RCTM_User, time=1, value=0}]}
     }}
    Scale Sprite Size Mode=::NewEnumerator3
   )
   &ScaleColor(
    Scale Alpha=&Multiply_Float(
     A=1
     B=&FloatFromCurve(
      FloatCurve=DataInterface() {propertyValues {Curve {keys [{interpMode=RCIM_Cubic, tangentMode=RCTM_User, time=0, value=1}, {interpMode=RCIM_Cubic, tangentMode=RCTM_User, time=1, value=0}]}}}
      Scale Curve=0.5
     )
    )
    Scale RGB=<5 5 5>
    ScaleA=true
    ScaleRGB=true
   )
   &SolveForcesAndVelocity(Clamp Velocity=false, Limit Acceleration=false)
  ]
  rendererClasses [&NiagaraSpriteRendererProperties]
  renderers [{rendererClass=&NiagaraSpriteRendererProperties}]
 }
 StaticRings {
  emitterUpdateScript [
   &EmitterState(Life Cycle Mode=::NewEnumerator1, Loop Behavior=::NewEnumerator1, Loop Duration Mode=::NewEnumerator1, UseLoopDelay=false)
   &SpawnBurst_Instantaneous(Use Spawn Probability=false, UseLoopCountLimit=false)
  ]
  particleSpawnScript [&InitializeParticle(
    Color=$User.TrackColor
    Mesh Scale Mode=::NewEnumerator3
    Mesh Uniform Scale=&Lerp_Float(A=&Lerp_Float(A=0.6825, B=1.35, Alpha=$User.IsDrum), B=0.5, Alpha=$User.IsAutoJammer)
    UsePositionOffset=false
    Write Material Random=true
   )]
  particleUpdateScript [
   &ParticleState(Kill Particles When Lifetime Has Elapsed=false, Let Infinitely Lived Particles Die When Emitter Deactivates=true)
   &UpdateMeshOrientation(
    Facing Direction=<0 1 0>
    Facing Interpolation=false
    Orientation Method=::NewEnumerator1
    Override Quat=false
    TransformFacing=false
    TransformUp=false
   )
   &SkeletalMeshLocation(
    MaskSpawnGroup=false
    Mesh Sampling Type=::NewEnumerator5
    Orientation Sampling=::NewEnumerator1
    Sampled Position Offset=&Lerp_Vector(A=&Lerp_Vector(A=<0 0 28.5>, B=<50 0 30>, Alpha=$User.IsDrum), B=<0 0 28.5>, Alpha=$User.IsAutoJammer)
    Skeletal Mesh=$System.Skeletal Mesh
    Socket Sampling Mode=::NewEnumerator5
   ) {module=SkeletalMeshLocation003}
   &ScaleColor(
    Scale RGB=&VectorFromFloat(Value=&FloatFromCurve(FloatCurve=DataInterface() {propertyValues {Curve {keys [
          {arriveTangent=9.0725, interpMode=RCIM_Cubic, leaveTangent=9.0725, tangentMode=RCTM_User, time=0, value=1}
          {arriveTangent=-2.5249, interpMode=RCIM_Cubic, leaveTangent=-2.5249, tangentMode=RCTM_User, time=0.05, value=3}
          {arriveTangent=-0.0405, interpMode=RCIM_Cubic, leaveTangent=-0.0405, tangentMode=RCTM_User, time=0.5, value=1}
         ]}}}))
    ScaleA=true
    ScaleRGB=true
   )
   &ScaleMeshSize(Scale Factor=&VectorFromFloat(Value=&Lerp_Float(
      A=&Lerp_Float(
       A=&FloatFromCurve(FloatCurve=$System.AnimCurve)
       B=&FloatFromCurve(FloatCurve=$System.AnimCurveDrum)
       Alpha=$User.IsDrum
      )
      B=&FloatFromCurve(FloatCurve=$System.AnimCurve)
      Alpha=$User.IsAutoJammer
     )))
  ]
  rendererClasses [&NiagaraMeshRendererProperties]
  renderers [
   {rendererClass=&NiagaraMeshRendererProperties, rendererIndex=0}
   {rendererClass=&NiagaraMeshRendererProperties, rendererIndex=1}
  ]
 }
 WavyRings {
  emitterUpdateScript [
   &EmitterState(Life Cycle Mode=::NewEnumerator1, Loop Behavior=::NewEnumerator1, Loop Duration Mode=::NewEnumerator1, UseLoopDelay=false)
   &SpawnBurst_Instantaneous(Use Spawn Probability=false, UseLoopCountLimit=false) {enabled=false}
   &SpawnRate(SpawnRate=2, Use Spawn Probability=false) {module=SpawnRate001}
  ]
  particleSpawnScript [&InitializeParticle(
    Color=$User.TrackColor
    Lifetime=1.5
    Mesh Scale Mode=::NewEnumerator3
    Mesh Uniform Scale=&Lerp_Float(A=&Lerp_Float(A=0.6825, B=1.35, Alpha=$User.IsDrum), B=0.5, Alpha=$User.IsAutoJammer)
    UsePositionOffset=false
    Write Material Random=true
   )]
  particleUpdateScript [
   &SolveForcesAndVelocity(Clamp Velocity=false, Limit Acceleration=false)
   &ParticleState()
   &UpdateMeshOrientation(
    Facing Direction=<0 -1 0>
    Facing Interpolation=false
    Orientation Method=::NewEnumerator1
    Override Quat=false
    TransformFacing=false
    TransformUp=false
   ) {module=UpdateMeshOrientation001}
   &ScaleMeshSize(Scale Factor=&VectorFromCurve(VectorCurve=DataInterface() {propertyValues {XCurve {keys [<at 0 1>,<at 1 1.15>]}, YCurve {keys [<at 0 1>,<at 1 1.15>]}, ZCurve {keys [<at 0 1>,<at 1 1.15>]}}}))
   &SkeletalMeshLocation(
    MaskSpawnGroup=false
    Mesh Sampling Type=::NewEnumerator5
    Orientation Sampling=::NewEnumerator1
    Sampled Position Offset=&AddVectorToPosition(
     Position=&Lerp_Vector(A=&Lerp_Vector(A=<0 0 30>, B=<50 0 30>, Alpha=$User.IsDrum), B=<0 0 30>, Alpha=$User.IsAutoJammer)
     Vector=&Lerp_Vector(A=<0 0 0>, B=<0 0 -30>, Alpha=$Particles.NormalizedAge)
    )
    Skeletal Mesh=$System.Skeletal Mesh
    Socket Sampling Mode=::NewEnumerator5
   ) {module=SkeletalMeshLocation003}
   &DynamicMaterialParameters(
    Index 0 Param 1=&Multiply_Float(
     A=$User.LoopAmplitudeMultiply
     B=&FloatFromCurve(
      FloatCurve=DataInterface() {propertyValues {Curve {keys [{interpMode=RCIM_Cubic, tangentMode=RCTM_User, time=0, value=0}, {interpMode=RCIM_Cubic, tangentMode=RCTM_User, time=0.25, value=0.75}]}}}
      Scale Curve=0.5
     )
    )
    Index 0 Param 3=1
    Index 1 Param 1=&MakeFloatFromVector4(Vector 4=$User.4BandReactivity, Channel=::NewEnumerator0)
    Index 1 Param 2=&MakeFloatFromVector4(Vector 4=$User.4BandReactivity, Channel=::NewEnumerator1)
    Index 1 Param 3=&MakeFloatFromVector4(Vector 4=$User.4BandReactivity, Channel=::NewEnumerator0)
    Index 1 Param 4=&MakeFloatFromVector4(Vector 4=$User.4BandReactivity, Channel=::NewEnumerator3)
    Module.Index0Param0Write=true
    Module.Index0Param1Write=true
    Module.Index0Param2Write=true
    Module.Index0Param3Write=true
    Module.Index1Param0Write=true
    Module.Index1Param1Write=true
    Module.Index1Param2Write=true
    Module.Index1Param3Write=true
    Write Parameter Index 1=true
   )
   &ScaleColor(
    Scale Alpha=&FNTimeOfDayScalar(Night Scale=0.75, Morning Scale=0.75, Evening Scale=0.75)
    ScaleA=true
    ScaleRGB=true
   )
   &ScaleColor(
    Scale Alpha=&FloatFromCurve(
     FloatCurve=DataInterface() {propertyValues {Curve {keys [
         {arriveTangent=1.2172, interpMode=RCIM_Cubic, leaveTangent=1.2172, tangentMode=RCTM_User, time=0, value=0}
         {arriveTangent=-2.5315, interpMode=RCIM_Cubic, leaveTangent=-2.5315, tangentMode=RCTM_User, time=0.2, value=0.9847}
         {arriveTangent=-0.0405, interpMode=RCIM_Cubic, leaveTangent=-0.0405, tangentMode=RCTM_User, time=1, value=0}
        ]}}}
     Scale Curve=&Add_Float(
      A=0.25
      B=&Divide_Float(
       A=&Add_Float(
        A=&MakeFloatFromVector4(Vector 4=$User.4BandReactivity, Channel=::NewEnumerator0)
        B=&Add_Float(
         A=&MakeFloatFromVector4(Vector 4=$User.4BandReactivity, Channel=::NewEnumerator1)
         B=&Add_Float(
          A=&MakeFloatFromVector4(Vector 4=$User.4BandReactivity, Channel=::NewEnumerator2)
          B=&MakeFloatFromVector4(Vector 4=$User.4BandReactivity, Channel=::NewEnumerator3)
         )
        )
       )
       B=4
      )
     )
    )
    ScaleA=true
    ScaleRGB=true
   ) {module=ScaleColor001}
  ]
  rendererClasses [&NiagaraMeshRendererProperties]
  renderers [{rendererClass=&NiagaraMeshRendererProperties, rendererIndex=0}]
 }
}
summary {
 systemName=NS_Jam_Character_Reactivity_LowProfile
 emitters [
  {emitterName=ReactiveWave, rendererClasses [&NiagaraMeshRendererProperties]}
  {emitterName=StaticRings, rendererClasses [&NiagaraMeshRendererProperties]}
  {emitterName=WavyRings, rendererClasses [&NiagaraMeshRendererProperties]}
  {emitterName=SquareParticles, rendererClasses [&NiagaraSpriteRendererProperties]}
  {emitterName=Burst, rendererClasses [&NiagaraSpriteRendererProperties]}
  {emitterName=CylinderGlows, rendererClasses [&NiagaraMeshRendererProperties]}
 ]
 userVariables [
  User.4BandReactivity {type=Vector4f, description="", defaultValue {struct=Vector4f, value=<0 0 0 0>}}
  User.IsAutoJammer {type=NiagaraFloat, defaultValue=0, description=""}
  User.IsDrum {type=NiagaraFloat, defaultValue=0, description=""}
  User.LoopAmplitudeMultiply {type=NiagaraFloat, defaultValue=0, description=""}
  User.TrackColor {type=LinearColor, description="", defaultValue {struct=LinearColor, value=<rgba 0.1867 1 0.1363 1>}}
 ]
}
